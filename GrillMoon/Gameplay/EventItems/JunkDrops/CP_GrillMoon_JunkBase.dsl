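; Base Blueprint for GrillMoon junk drop event items, which drop a gold (Wads) resource pickup when destroyed.
; On server-side death it spawns the Athena_WadsItemData pickup with the quantity from GoldToSpawn.
; The pickup appears at the actor's location plus the PickUpSpawnOffest vector, using the actor's forward vector as its direction.
; The construction script is empty.
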(event EventOnDeathServer
  (params (Damage :type Float%20%28single-precision%29) (DamageTags :type Gameplay%20Tag%20Container%20Structure%20%28by%20ref%29) (Momentum :type Vector :default "0, 0, 0") (HitInfo :type Hit%20Result%20Structure%20%28by%20ref%29) (InstigatedBy :type Controller%20Object%20Reference) (DamageCauser :type Actor%20Object%20Reference) (EffectContext :type Gameplay%20Effect%20Context%20Handle%20Structure))
  (Inventory|SpawnPickupInWorld "/Game/Items/ResourcePickups/Athena_WadsItemData.Athena_WadsItemData" (Variables|Default|GetGoldToSpawn) (+ (Transformation|GetActorLocation) (Variables|Default|GetPickUpSpawnOffest)) (Transformation|GetActorForwardVector)))

(fn ConstructionScript)
